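; Boom Box consumable actor for Athena that takes its instigator's team and binds to the destruction of its attach parent.
; On BeginPlay it briefly plays its mesh animation until the beat starts, when OnRep_b_BeatStarted seeks past the intro and loops it.
; Server death destroys the attach parent, whose OnDestroyed event triggers the death sound and a timeline that scales the mesh, spawns particles, then destroys the actor.
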(event EventBeginPlay
  (bind _returnvalue (Game|GetInstigator))
  (if (Utilities|IsValid _returnvalue :!class /Script/Engine.KismetSystemLibrary)
    (CallFunction|SetTeam (Team|GetActorTeam _returnvalue))
    (CallFunction|BoundEvents)
    (if (not (Variables|Default|Getb_BeatStarted))
      (Components|Animation|Play (Variables|Default|GetSkeletalMesh) false)
      (Utilities|FlowControl|Delay 0.71)
      (Components|Animation|Stop (Variables|Default|GetSkeletalMesh)))))

(event Custom|BoundEvents
  (bind _attachparent (Variables|Default|GetAttachParent))
  (Utilities|IsValid _attachparent
    (:Is%20Valid
      (Game|BindEventtoOnDestroyed _attachparent (EventDispatchers|CreateEvent :!event Custom|DestroySelf))
      (Development|PrintString (Utilities|GetDisplayName _attachparent)))))

(event EventOnDeathServer
  (params (Damage :type Float%20%28single-precision%29) (DamageTags :type Gameplay%20Tag%20Container%20Structure%20%28by%20ref%29) (Momentum :type Vector :default "0, 0, 0") (HitInfo :type Hit%20Result%20Structure%20%28by%20ref%29) (InstigatedBy :type Controller%20Object%20Reference) (DamageCauser :type Actor%20Object%20Reference) (EffectContext :type Gameplay%20Effect%20Context%20Handle%20Structure))
  (bind _attachparent (Variables|Default|GetAttachParent))
  (Utilities|IsValid _attachparent
    (:Is%20Valid
      (Actor|DestroyActor _attachparent))))

(event Custom|DestroySelf
  (params (DestroyedActor :type Actor%20Object%20Reference))
  (Development|PrintString "Death" true true "(R=0.553492,G=0.000000,B=1.000000,A=1.000000)" 5.0)
  (CallFunction|SpawnDeathFX))

(event Custom|SpawnDeathFX
  (Audio|PlaySoundatLocation "/Game/Sounds/Fort_Gadgets/Boombox/Boombox_Destroy_Cue.Boombox_Destroy_Cue" (Transformation|GetActorLocation))
  (bind (_direction _size) (Timeline|Death Play)
    (:Update
      (Transformation|SetRelativeScale3D (Variables|Default|GetSkeletalMesh) (Math|Vector|Lerp(Vector) "0, 0, 0" "1.000000,1.000000,1.000000" _size)))
    (:Finished
      (Rendering|SetActorHiddenInGame self true)
      (Actor|DestroyActor))
    (:SpawnDeathFX
      (Development|PrintString "PlayDeathFX" true true "(R=0.834767,G=0.000000,B=1.000000,A=1.000000)" 5.0)
      (Effects|SpawnEmitteratLocation "/Game/Athena/Items/Consumables/BoomBox/P_Athena_BoomBox_Death.P_Athena_BoomBox_Death" (Transformation|GetActorLocation)))))

(event Collision|EventActorBeginOverlap
  (params (OtherActor :type Actor%20Object%20Reference)))

(event EventTick
  (params (DeltaSeconds :type Float%20%28single-precision%29)))

(fn OnRep_b_BeatStarted
  (bind _skeletalmesh (Variables|Default|GetSkeletalMesh))
  (if (Variables|Default|Getb_BeatStarted)
    (Components|Animation|SetPosition _skeletalmesh 0.745 true)
    (Components|Animation|Play _skeletalmesh true)
    (Development|PrintString "GOT EM")))

(fn ConstructionScript)
