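; Thrown Boom Box projectile that lands on a surface and pulses explosions to a music beat, launching nearby vehicles.
; On stop, the server spawns an attached BGA, attaches to what it hit, and derives beat start and loop position from server world time.
; A replicated flag starts music and repeating blasts with cue-driven shake.
; When its support is destroyed, it rebinds to overlapping buildings or sphere-traces down to drop; tuning comes from AthenaGameData.
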(event EventBeginPlay
  (|CallParentFunction))

(event Collision|EventActorBeginOverlap
  (params (OtherActor :type Actor%20Object%20Reference)))

(event EventTick
  (params (DeltaSeconds :type Float%20%28single-precision%29))
  (Development|PrintString (Utilities|String|ToString(Vector) (Transformation|GetVelocity))))

(event EventOnStop
  (params (Hit :type Hit%20Result%20Structure%20%28by%20ref%29))
  (bind _beattime (Variables|Default|GetBeatTime))
  (bind _returnvalue (Game|GetGameState))
  (bind _returnvalue_2 (GameState|GetServerWorldTimeSeconds _returnvalue))
  (bind _remainder (Math|Float|Division(WholeandRemainder) _returnvalue_2 _beattime))
  (bind _beatmultiplierforstart (Variables|Default|GetBeatMultiplierForStart))
  (bind _returnvalue_3 (Math|Conversions|ToFloat(Integer) _beatmultiplierforstart))
  (bind _returnvalue_4 (* _beattime _remainder))
  (bind _returnvalue_5 (- _beattime _returnvalue_4))
  (bind (_remainder_2 _returnvalue_6) (Math|Float|Division(WholeandRemainder) (+ _returnvalue_2 (+ _returnvalue_5 (* (Variables|Default|GetBeatTime) _returnvalue_3))) _beattime))
  (bind (_bblockinghit _binitialoverlap _time _distance _location _impactpoint _normal _impactnormal _physmat _hitactor) (Collision|BreakHitResult Hit))
  (bind _beattime_2 (Variables|Default|GetBeatTime))
  (bind _returnvalue_7 (* _beattime_2 _returnvalue_3))
  (bind _returnvalue_8 (+ _returnvalue_5 _returnvalue_7))
  (|CallParentFunction Hit)
  (if (Networking|IsServer)
    (Utilities|FlowControl|DoOnce false
      (:Completed
        ;[1 1 1] change the lifespan after this hits the ground!
        (bind _outxy (DataTable|EvaluateCurveTableRow "/Game/Athena/Balance/DataTables/AthenaGameData.AthenaGameData" "Default.Boombox.MaxDuration")
          (:RowFound
            (Variables|Default|SetDuration _outxy))
          (:RowNotFound
            (Variables|Default|SetDuration 25.0)))
        (Variables|Projectile|SetMaxLifespanOnStop (Variables|Default|GetDuration))
        (Development|PrintString (Utilities|String|ToString(Float) _returnvalue_2) false false "(R=0.060894,G=1.000000,B=0.000000,A=1.000000)" 3e+01)
        (bind _output_get (Variables|Default|SetCurrentBeatInTheLoop (Math|Float|Truncate (Math|Float|Division(WholeandRemainder) (Math|Conversions|ToFloat(Integer) _returnvalue_6) (Math|Conversions|ToFloat(Integer) (Variables|Default|GetBeatsPerLoop))))))
        (Variables|Default|SetTimeUntilMusicStarts _returnvalue_8)
        (Development|PrintString (Utilities|String|ToString(Integer) _output_get) false false "(R=0.000000,G=0.660000,B=1.000000,A=1.000000)" 25.0)
        (Utilities|Time|SetTimerbyEvent (EventDispatchers|CreateEvent :!event Custom|DropTheBeat) (Variables|Default|GetTimeUntilMusicStarts))
        (Transformation|SetWorldRotation (Variables|Collision|GetCapsuleComponent) (Math|Rotator|MakeRotator 0.0 0.0 (.yaw (Transformation|GetWorldRotation (Variables|Collision|GetCapsuleComponent)))) false false)
        (CallFunction|SpawnBGA)
        (Utilities|Casting|CastToFortVehicleInterface _hitactor)
        (Transformation|AttachActorToActor self _hitactor "None" "KeepWorld" "KeepWorld" "KeepWorld" true)))
    (else
      (Audio|SpawnSoundAttached "/Game/Sounds/Fort_Gadgets/Boombox/Boombox_Activate_Cue.Boombox_Activate_Cue" (Variables|Default|GetAudioMusic))
      (Development|SetHiddeninGame (Variables|Default|GetSM_BoomBox_InHand) true false))))

(event Custom|BaseDiedEvent
  (params (DamagedActor :type Actor%20Object%20Reference) (Damage :type Float%20%28single-precision%29) (InstigatedBy :type Controller%20Object%20Reference) (DamageCauser :type Actor%20Object%20Reference) (HitLocation :type Vector) (FHitComponent :type Primitive%20Component%20Object%20Reference) (BoneName :type Name) (Momentum :type Vector))
  ;[1 1 1] Handle if the object this was on was destroyed
  (bind (_outactors _returnvalue) (Collision|SphereOverlapActors (Transformation|GetActorLocation) 16.0 (Utilities|Array|MakeArray "ObjectTypeQuery1") "/Script/FortniteGame.BuildingActor" (Utilities|Array|MakeArray self 0)))
  (if _returnvalue
    (bind _output_get (Variables|Default|SetFoundBuildingOnDied _outactors))
    (bind _asbuilding_actor (Utilities|Casting|CastToBuildingActor (Utilities|Array|Get(acopy) _output_get))
      (:then
        (EventDispatchers|AssignOnDied _asbuilding_actor (EventDispatchers|CreateEvent :!event Custom|BaseDiedEvent)))
      (:CastFailed
        (Game|BindEventtoOnDestroyed (Utilities|Array|Get(acopy) _output_get) (EventDispatchers|CreateEvent :!event Custom|ObjectDiedEvent))))
    (else
      (Variables|Default|SetbResumeSimulation true)
      (Networking|ForceNetUpdate))))

(event Collision|EventHit
  (params (MyComp :type Primitive%20Component%20Object%20Reference) (Other :type Actor%20Object%20Reference) (OtherComp :type Primitive%20Component%20Object%20Reference) (bSelfMoved :type Boolean) (HitLocation :type Vector :default "0, 0, 0") (HitNormal :type Vector :default "0, 0, 0") (NormalImpulse :type Vector :default "0, 0, 0") (Hit :type Hit%20Result%20Structure%20%28by%20ref%29))
  ; Event Hit
  ; When projectile hits a surface for the first time
  (CallFunction|CheckDeploymentSurface HitNormal Hit))

(event Custom|ObjectDiedEvent
  (params (DestroyedActor :type Actor%20Object%20Reference))
  ;[1 1 1] Handle if the object this was on was destroyed
  (bind (_outactors _returnvalue) (Collision|SphereOverlapActors (Transformation|GetActorLocation) 16.0 (Utilities|Array|MakeArray "ObjectTypeQuery1") "/Script/FortniteGame.BuildingActor" (Utilities|Array|MakeArray self 0)))
  (if _returnvalue
    (bind _output_get (Variables|Default|SetFoundBuildingOnDied _outactors))
    (bind _asbuilding_actor (Utilities|Casting|CastToBuildingActor (Utilities|Array|Get(acopy) _output_get))
      (:then
        (EventDispatchers|AssignOnDied _asbuilding_actor (EventDispatchers|CreateEvent :!event Custom|BaseDiedEvent)))
      (:CastFailed
        (Game|BindEventtoOnDestroyed (Utilities|Array|Get(acopy) _output_get) (EventDispatchers|CreateEvent :!event Custom|ObjectDiedEvent))))
    (else
      (Variables|Default|SetbResumeSimulation true)
      (Networking|ForceNetUpdate))))

(event Custom|CheckDeploymentSurface
  (params (InVec :type Vector) (Hit :type Hit%20Result%20Structure%20%28by%20ref%29))
  (bind (_bblockinghit _binitialoverlap _time _distance _location _impactpoint _normal _impactnormal _physmat _hitactor) (Collision|BreakHitResult Hit))
  (bind _projectilemovementcomponent (Variables|Movement|GetProjectileMovementComponent))
  ;[1 1 1] If the floor you are on is destroyed, start falling!
  (Utilities|FlowControl|Sequence
    (:then_0
      (Effects|SpawnEmitteratLocation "/Game/Athena/Items/Consumables/BoomBox/P_Athena_BoomBox_Deploy.P_Athena_BoomBox_Deploy" _impactpoint))
    (:then_1
      (if (> (.z InVec) 0.71)
        (bind _returnvalue (CallFunction|WillBounceOffTarget Hit))
        (if _returnvalue
          ;[1 1 1] If you hit a target that isn't valid
          (Variables|/Script/Engine.MovementComponent|SetVelocity (Math|Vector|vector+vector (Math|Vector|MirrorVectorbyNormal (* (Variables|/Script/Engine.MovementComponent|GetVelocity (Variables|Movement|GetProjectileMovementComponent)) 2.0) "0, 0,1.000000") (Math|Vector|MakeVector 0.0 0.0 (* (Math|Vector|VectorLength (Variables|/Script/Engine.MovementComponent|GetVelocity (Variables|Movement|GetProjectileMovementComponent))) 0.25))) (Variables|Movement|GetProjectileMovementComponent))
          (else
            ;; To avoid non vertical velocity upon resuming simulation
            (Variables|/Script/Engine.ProjectileMovementComponent|SetbShouldBounce false _projectilemovementcomponent)
            (Game|Components|ProjectileMovement|StopSimulating _projectilemovementcomponent Hit)
            (bind _asbuilding_actor (Utilities|Casting|CastToBuildingActor _hitactor)
              (:then
                (EventDispatchers|AssignOnDied _asbuilding_actor (EventDispatchers|CreateEvent :!event Custom|BaseDiedEvent)))
              (:CastFailed
                (Game|BindEventtoOnDestroyed _hitactor (EventDispatchers|CreateEvent :!event Custom|ObjectDiedEvent))))))))))

(event Custom|SpawnBGA
  (Utilities|FlowControl|DoOnce false
    (:Completed
      (bind _returnvalue (Game|SpawnBuildingActorfromClass "/Game/Athena/Items/Consumables/BoomBox/BGA_Athena_BoomBox.BGA_Athena_BoomBox_C" (Transformation|GetActorTransform self) (Variables|Actor|GetInstigator) false self 0 (Variables|Actor|GetInstigator)))
      (bind _output_get (Variables|Default|SetSpawnedBGA _returnvalue))
      (Transformation|AttachActorToActor _output_get self "None" "SnapToTarget" "SnapToTarget" "SnapToTarget" true)
      (Delivery|SetNoiseMakingInstigatorOverride _output_get))))

(event Custom|DropTheBeat
  ;[1 1 1] Drop the beat
  (Variables|Default|Setb_DroppingTheBeat true)
  (Networking|ForceNetUpdate))

(event Custom|ExplodeOnTheBeat
  ;[1 1 1] Explode on the beat
  (Delivery|DoExplosion)
  (Development|PrintString "BEAT" false false)
  (GameplayCue|ExecuteGameplayCue%28Burst%29Local self (Variables|Default|GetBoomCue) (Ability|GameplayCue|MakeGameplayCueParameters :NormalizedMagnitude (Variables|Default|GetInnerRadius) :RawMagnitude (Variables|Default|GetOuterRadius) :EffectContext "()" :AggregatedSourceTags "(GameplayTags=)" :AggregatedTargetTags "(GameplayTags=)" :Location "0,0,0" :Normal "0,0,0" :Instigator (Game|GetInstigator) :EffectCauser self :GameplayEffectLevel 1 :AbilityLevel (Math|Conversions|ToInteger(Boolean) (Variables|Default|GetShouldCameraShake)) :bReplicateLocationWhenUsingMinimalRepProxy false))
  (Fortnite|ShowSoundIndicatorLocal :Location (Transformation|GetActorLocation) :MaxAudibleDistance 2.8e+03 :IndicatorType "Gunshot" :Instigator self :Tint "(R=0.000000,G=0.000000,B=0.000000,A=0.000000)" :AttenuationMaxRangeMultiplier 1.0 :AttenuationIntensityMultiplier 1.0))

(event EventOnExploded
  (params (HitActors :type Array%20of%20Actor%20Object%20References) (HitResults :type Array%20of%20Hit%20Result%20Structures))
  (bind _self self)
  (for _array_element HitActors
    (bind _asfort_vehicle_interface (Utilities|Casting|CastToFortVehicleInterface _array_element))
    (if (Utilities|Equal(Object) (Actor|GetAttachParentActor) (Vehicle|GetActor _asfort_vehicle_interface))
      _
      (elif (< (Transformation|GetDistanceTo (Vehicle|GetActor _asfort_vehicle_interface) _self) (Variables|Default|GetMinDistanceToKnock))
        ;[1 1 1] Launch the vehicle
        (Audio|BroadcastAbilityImpactatLocation 0 (Vehicle|GetActor _asfort_vehicle_interface) (+ (Math|Vector|MakeVector 0.0 0.0 (Math|Float|MapRangeUnclamped (Transformation|GetDistanceTo (Vehicle|GetActor _asfort_vehicle_interface) _self) 0.0 1.3e+03 (Variables|Default|GetMaxLaunchPower) (Variables|Default|GetMinLaunchPower))) (Vehicle|GetVehicleVelocity _asfort_vehicle_interface)) 7.5e+02)))))

(fn OnRep_b_DroppingTheBeat
  (bind _beattime (Variables|Default|GetBeatTime))
  (bind _currentbeatintheloop (Variables|Default|GetCurrentBeatInTheLoop))
  (bind _returnvalue (* _currentbeatintheloop _beattime))
  (bind _spawnedbga (Variables|Default|GetSpawnedBGA))
  (if (Variables|Default|Getb_DroppingTheBeat)
    (Utilities|Time|SetTimerbyEvent (EventDispatchers|CreateEvent :!event ExplodeOnTheBeat) (Variables|Default|GetBeatTime) true)
    (Development|PrintString "drop the" false false)
    (Audio|Components|Audio|Play (Variables|Default|GetAudioMusic) _returnvalue)
    (Development|PrintString (Utilities|String|ToString(Float) _returnvalue) true true "(R=1.000000,G=0.031642,B=0.221980,A=1.000000)" 15.0)
    (CallFunction|ExplodeontheBeat)
    (Utilities|IsValid _spawnedbga
      (:Is%20Valid
        (Variables|/Game/Athena/Items/Consumables/BoomBox/BGA_Athena_BoomBox.BGA_Athena_BoomBox_C|Setb_BeatStarted true _spawnedbga)
        (Networking|ForceNetUpdate _spawnedbga)))))

(fn OnRep_bResumeSimulation
  (bind (_returnvalue_location _returnvalue_rotation) (Math|Transform|BreakTransform (Transformation|GetActorTransform)))
  (bind _projectilemovementcomponent (Variables|Movement|GetProjectileMovementComponent))
  ; Resume simulation after supporting structure is destroyed:
  ; Running a sphere trace below the actor to find next surface to drop to, then interpolate to that surface.
  (if (Variables|Default|GetbResumeSimulation)
    (bind (_outhit _returnvalue) (Collision|SphereTraceByChannel _returnvalue_location (- _returnvalue_location "0.000000,0.000000,10000.000000") 22.0 "TraceTypeQuery1" false (Utilities|Array|MakeArray (Variables|Default|GetSpawnedBGA)) "ForDuration"))
    (if _returnvalue
      (Components|Movement|SetUpdatedComponent _projectilemovementcomponent (Variables|Collision|GetCapsuleComponent))
      (Game|Components|ProjectileMovement|Interpolation|MoveInterpolationTarget _projectilemovementcomponent (.Location (Collision|BreakHitResult _outhit)) _returnvalue_rotation)
      (CallFunction|CheckDeploymentSurface (.Normal (Collision|BreakHitResult _outhit)) _outhit)
      (else
        (Actor|DestroyActor)))))

(fn ConstructionScript
  ;[1 1 1] DEFAULT LIFESPAN, dont change this until the projectile lands!
  (Actor|SetLifeSpan (Variables|Default|GetDuration))
  (Utilities|FlowControl|Sequence
    (:then_0
      (bind _outxy (DataTable|EvaluateCurveTableRow "/Game/Athena/Balance/DataTables/AthenaGameData.AthenaGameData" "Default.Boombox.BeatTime")
        (:RowFound
          (Variables|Default|SetBeatTime _outxy))))
    (:then_1
      (bind _outxy_2 (DataTable|EvaluateCurveTableRow "/Game/Athena/Balance/DataTables/AthenaGameData.AthenaGameData" "Default.Boombox.BeatMultiplierForStart")
        (:RowFound
          (Variables|Default|SetBeatMultiplierForStart (Math|Float|Truncate _outxy_2)))))
    (:then_2
      (bind _outxy_3 (DataTable|EvaluateCurveTableRow "/Game/Athena/Balance/DataTables/AthenaGameData.AthenaGameData" "Default.Boombox.BeatsPerLoop")
        (:RowFound
          (Variables|Default|SetBeatsPerLoop (Math|Float|Truncate _outxy_3)))))
    (:then_3
      (bind _outxy_4 (DataTable|EvaluateCurveTableRow "/Game/Athena/Balance/DataTables/AthenaGameData.AthenaGameData" "Default.Boombox.ShouldBounceVehicles")
        (:RowFound
          (Variables|Default|SetShouldBounceVehicles (Math|Conversions|ToBoolean(Integer) (Math|Float|Truncate _outxy_4))))))
    (:then_4
      (bind _outxy_5 (DataTable|EvaluateCurveTableRow "/Game/Athena/Balance/DataTables/AthenaGameData.AthenaGameData" "Default.Boombox.MinVehicleLaunchPower")
        (:RowFound
          (Variables|Default|SetMinLaunchPower _outxy_5))))
    (:then_5
      (bind _outxy_6 (DataTable|EvaluateCurveTableRow "/Game/Athena/Balance/DataTables/AthenaGameData.AthenaGameData" "Default.Boombox.MaxVehicleLaunchPower")
        (:RowFound
          (Variables|Default|SetMaxLaunchPower _outxy_6))))
    (:then_6
      (bind _outxy_7 (DataTable|EvaluateCurveTableRow "/Game/Athena/Balance/DataTables/AthenaGameData.AthenaGameData" "Default.Boombox.ShouldScreenShake")
        (:RowFound
          (Variables|Default|SetShouldCameraShake (Math|Conversions|ToBoolean(Integer) (Math|Float|Truncate _outxy_7))))))
    (:then_7
      (bind _outxy_8 (DataTable|EvaluateCurveTableRow "/Game/Athena/Balance/DataTables/AthenaGameData.AthenaGameData" "Default.Boombox.ScreenShakeMinRange")
        (:RowFound
          (Variables|Default|SetInnerRadius _outxy_8))))
    (:then_8
      (bind _outxy_9 (DataTable|EvaluateCurveTableRow "/Game/Athena/Balance/DataTables/AthenaGameData.AthenaGameData" "Default.Boombox.ScreenShakeMaxRange")
        (:RowFound
          (Variables|Default|SetOuterRadius _outxy_9))))))

(fn WillBounceOffTarget
  (params (Hit :type Hit%20Result%20Structure%20%28by%20ref%29))
  (bind (_bblockinghit _binitialoverlap _time _distance _location _impactpoint _normal _impactnormal _physmat _hitactor _hitcomponent) (Collision|BreakHitResult Hit))
  (if (Components|ComponentHasTag _hitcomponent "NoStick")
    (return true)
    (else
      (Utilities|Casting|CastToBuildingGameplayActor _hitactor
        (:then
          (return true))
        (:CastFailed
          (return false))))))
